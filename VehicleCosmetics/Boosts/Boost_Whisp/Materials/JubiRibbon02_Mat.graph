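# Translucent unlit Niagara ribbon material for the Whisp vehicle boost cosmetic.
# A leftward-panned gradient texture, scaled by the GradientGlow dynamic parameter, is multiplied by a squared, panning fire dissolve texture and an inverted-U falloff along the ribbon.
# The result feeds a shared emissive adjustment function, and a boosted copy times particle alpha is saturated, apparently for opacity.
# The Dissolve dynamic parameter is named but unused.
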Material /VehicleCosmetics/Boosts/Boost_Whisp/Materials/JubiRibbon02_Mat
graph {
 properties {
  bAutomaticallySetUsageInEditor=false
  bUsedWithNiagaraRibbons=true
  blendMode=BLEND_Translucent
  shadingModel=MSM_Unlit
  usedShadingModels=Unlit
 }
 nodes [
  MaterialExpressionVertexColor_2 = ParticleColor()
  n1 = Panner(Coordinate=TextureCoordinate()) {speedX=-0.8}
  n2 = TextureSample(UVs=Panner(Coordinate=TextureCoordinate()) {speedX=-0.8, speedY=0.1}) {texture=/VehicleCosmetics/Boosts/Boost_Whisp/Textures/Dissolve_Fire_M_T}
  n3 = TextureCoordinate()
  n4 = Power(Base=OneMinus(None=ComponentMask(None=@n3) {r=true}), Exp=Constant() {r=1})
  n5 = Multiply(
   A=Multiply(
    A=TextureSample(UVs=@n1) {texture=/VehicleCosmetics/Boosts/Boost_Whisp/Textures/JubiGradient_T}.RGB
    B=DynamicParameter() {paramNames [Dissolve,GradientGlow,Param3,Param4]}.GradientGlow
   )
   B=Multiply(A=@n2.RGB, B=Multiply(A=@n2.RGB, B=@n4))
  )
  n6 = Multiply(A=@n5, B=Constant() {r=2.5})
  MaterialExpressionConstantClamp_2 = Saturate(None=Multiply(A=@MaterialExpressionVertexColor_2.A, B=@n6))
  MaterialFunctionCall(Emissive=@n5) {materialFunction=&MF_FX_EmissiveAdjustments}
  Reroute(None=@n6)
 ]
}
