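# Translucent, unlit, two-sided Niagara ribbon material for the Whisp boost in VehicleCosmetics.
# Emissive multiplies a panning gradient and GradientGlow by summed sine/cosine-pulsed sphere masks, a bright space texture and the main ribbon texture.
# Scrolling noise scaled by Distort warps the main ribbon's UVs.
# Opacity smoothsteps that ribbon, faded along U, against a Dissolve dynamic parameter, then applies particle alpha and a soft V-edge falloff.
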Material /VehicleCosmetics/Boosts/Boost_Whisp/Materials/JubiRibbon_Mat
graph {
 properties {
  bAutomaticallySetUsageInEditor=false
  bUsedWithNiagaraRibbons=true
  blendMode=BLEND_Translucent
  shadingModel=MSM_Unlit
  twoSided=true
  usedShadingModels=Unlit
 }
 nodes [
  n1 = TextureSample(UVs=Panner(Coordinate=TextureCoordinate()) {speedX=-0.8}) {texture=/VehicleCosmetics/Boosts/Boost_Whisp/Textures/JubiGradient_T}
  n2 = DynamicParameter() {paramNames [Dissolve,GradientGlow,Distort,Param4]}
  MaterialExpressionConstantClamp_5 = Saturate(None=Sine(None=Time()) {period=0.8})
  n3 = TextureSample(UVs=Panner(Coordinate=TextureCoordinate() {uTiling=1.5, vTiling=1.5}) {speedX=-0.8, speedY=0.05}) {texture=/DealershipSharedLibrary/FX/Textures/Gradient_Masks/T_VCVFX_SphereMaskTimeOffset_A}
  n4 = Multiply(A=@MaterialExpressionConstantClamp_5, B=@n3.R)
  MaterialExpressionConstantClamp_6 = Saturate(None=Cosine(None=Time()) {period=0.8})
  n5 = TextureSample(UVs=Panner(Coordinate=TextureCoordinate() {uTiling=2, vTiling=2}) {speedX=-0.45, speedY=0.2}) {texture=&T_VCVFX_SphereMaskTimeOffset_A}
  n6 = Multiply(A=@MaterialExpressionConstantClamp_6, B=@n5.R)
  n7 = Multiply(A=@MaterialExpressionConstantClamp_5, B=@n3.B)
  n8 = Multiply(A=@n5.B, B=@MaterialExpressionConstantClamp_6)
  MaterialExpressionConstantClamp_4 = Saturate(None=Add(A=Add(A=@n4, B=@n6), B=Add(A=@n7, B=@n8)))
  n9 = TextureCoordinate() {uTiling=3, vTiling=2}
  n10 = Multiply(
   A=Constant() {r=20}
   B=TextureSample(UVs=Panner(Coordinate=@n9) {speedX=-0.45}) {texture=/VehicleCosmetics/Boosts/Boost_Spacetear/Textures/T_VCVFX_boost_Space_Tear_Space}.RGB
  )
  n11 = Add(A=@MaterialExpressionConstantClamp_4, B=@n10)
  n12 = TextureCoordinate() {uTiling=6, vTiling=1.35}
  n13 = Panner(Coordinate=TextureCoordinate() {uTiling=0.75, vTiling=0.75}) {speedX=-0.6, speedY=0.5}
  n14 = Multiply(
   A=ComponentMask(None=TextureSample(UVs=@n13) {texture=/DealershipSharedLibrary/FX/Textures/Noise/T_VCVFX_TrailMask_RGB}.RGB) {g=true, r=true}
   B=@n2.Distort
  )
  n15 = TextureSampleParameter2D(UVs=Panner(Coordinate=Add(A=@n12, B=@n14)) {speedX=-3.5}) {parameterName=MainRibbonTexture, texture=/VehicleCosmetics/Boosts/Boost_Whisp/Textures/ShockRibbon_RGB_T}
  MaterialFunctionCall(Emissive=Multiply(A=Multiply(A=@n1.RGB, B=@n2.GradientGlow), B=Add(A=@n11, B=@n15.G))) {materialFunction=&MF_FX_EmissiveAdjustments}
  n16 = OneMinus(None=ComponentMask(None=TextureCoordinate()) {r=true})
  n17 = Constant() {r=2}
  n18 = SmoothStep(Min=@n2.Dissolve, Max=Add(A=@n2.Dissolve, B=ScalarParameter() {parameterName=Max}), Value=Multiply(A=@n15.G, B=Power(Base=@n16, Exp=@n17)))
  MaterialExpressionVertexColor_3 = ParticleColor()
  n19 = ComponentMask(None=TextureCoordinate()) {g=true}
  n20 = Constant() {r=1}
  n21 = Multiply(A=Power(Base=OneMinus(None=@n19), Exp=@n20), B=Power(Base=@n19, Exp=@n20))
  n22 = Constant() {r=4}
  MaterialExpressionConstantClamp_3 = Saturate(None=Multiply(A=Multiply(A=@n18, B=@MaterialExpressionVertexColor_3.A), B=Multiply(A=@n21, B=@n22)))
  Multiply(A=@MaterialExpressionConstantClamp_3, B=@MaterialExpressionVertexColor_3.A)
  Multiply(A=@MaterialExpressionConstantClamp_4, B=@n10)
 ]
}
