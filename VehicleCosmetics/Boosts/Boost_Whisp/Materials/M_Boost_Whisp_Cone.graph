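# Unlit, two-sided translucent Niagara mesh-particle material, evidently for the cone of the Whisp vehicle boost.
# Two panning mask textures pulse on sine and cosine, reveal a brightened, scrolling Spacetear space texture, then add a shock-ribbon mask.
# A scrolling gradient scaled by dynamic GradientGlow multiplies this, with a clamped boost from camera angle against a fixed X vector, feeding emissive adjustments.
# The ribbon mask times particle alpha forms a separate saturated mask.
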Material /VehicleCosmetics/Boosts/Boost_Whisp/Materials/M_Boost_Whisp_Cone
graph {
 properties {
  bAutomaticallySetUsageInEditor=false
  bUsedWithNiagaraMeshParticles=true
  blendMode=BLEND_Translucent
  shadingModel=MSM_Unlit
  twoSided=true
  usedShadingModels=Unlit
 }
 nodes [
  MaterialExpressionMeshEmitterVertexColor_0 = ParticleColor()
  n1 = TextureCoordinate()
  MaterialExpressionConstantClamp_1 = Saturate(None=Multiply(A=TextureSample(UVs=@n1) {texture=/VehicleCosmetics/Boosts/Boost_Whisp/Textures/ShockRibbon_RGB_T}.B, B=Constant() {r=5}))
  MaterialExpressionConstantClamp_0 = Saturate(None=Multiply(A=@MaterialExpressionMeshEmitterVertexColor_0.A, B=@MaterialExpressionConstantClamp_1))
  n2 = CameraVectorWS()
  n3 = Constant3Vector() {constant=<rgba 1 0 0 0>}
  n4 = Multiply(
   A=Power(Base=DotProduct(A=@n2, B=@n3), Exp=ScalarParameter() {defaultValue=5, parameterName=Power}) {constExponent=5}
   B=ScalarParameter() {defaultValue=1, parameterName=Intensity}
  )
  n5 = TextureCoordinate()
  MaterialExpressionMeshEmitterDynamicParameter_0 = DynamicParameter() {paramNames [Param1,GradientGlow,BlackCore,Param4]}
  n6 = Multiply(
   A=TextureSample(UVs=Panner(Coordinate=@n5) {speedX=-0.8}) {texture=/VehicleCosmetics/Boosts/Boost_Whisp/Textures/JubiGradient_T}.RGB
   B=@MaterialExpressionMeshEmitterDynamicParameter_0.GradientGlow
  )
  MaterialExpressionConstantClamp_3 = Saturate(None=Sine(None=Time()) {period=0.8})
  n7 = TextureSample(UVs=Panner(Coordinate=TextureCoordinate() {uTiling=1.5, vTiling=1.5}) {speedX=-0.8, speedY=0.05}) {texture=/DealershipSharedLibrary/FX/Textures/Gradient_Masks/T_VCVFX_SphereMaskTimeOffset_A}
  n8 = Multiply(A=@MaterialExpressionConstantClamp_3, B=@n7.R)
  MaterialExpressionConstantClamp_4 = Saturate(None=Cosine(None=Time()) {period=0.8})
  n9 = TextureSample(UVs=Panner(Coordinate=TextureCoordinate() {uTiling=2, vTiling=2}) {speedX=-0.45, speedY=0.2}) {texture=&T_VCVFX_SphereMaskTimeOffset_A}
  n10 = Multiply(A=@MaterialExpressionConstantClamp_4, B=@n9.R)
  n11 = Multiply(A=@n7.B, B=@MaterialExpressionConstantClamp_3)
  n12 = Multiply(A=@MaterialExpressionConstantClamp_4, B=@n9.B)
  MaterialExpressionConstantClamp_2 = Saturate(None=Add(A=Add(A=@n8, B=@n10), B=Add(A=@n11, B=@n12)))
  n13 = Constant() {r=100}
  n14 = TextureSample(UVs=Panner(Coordinate=TextureCoordinate() {uTiling=3, vTiling=2}) {speedX=-0.45}) {texture=/VehicleCosmetics/Boosts/Boost_Spacetear/Textures/T_VCVFX_boost_Space_Tear_Space}
  n15 = Add(A=Multiply(A=@MaterialExpressionConstantClamp_2, B=Multiply(A=@n13, B=@n14.RGB)), B=@MaterialExpressionConstantClamp_1)
  MaterialFunctionCall(Emissive=Multiply(A=Clamp(None=@n4) {maxDefault=100, minDefault=1}, B=Multiply(A=@n6, B=@n15))) {materialFunction=&MF_FX_EmissiveAdjustments}
 ]
}
