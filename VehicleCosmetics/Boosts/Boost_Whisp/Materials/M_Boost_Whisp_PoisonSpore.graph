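# Translucent unlit particle material for the Whisp boost's poison spore effect, usable on Niagara and Cascade sprites.
# A procedural exponential radial gradient with a Radius parameter shapes a soft dot, with no textures.
# Particle color times the gradient feeds the shared emissive adjustments function.
# Particle alpha times the gradient and a camera depth fade feeds a depth fade function for opacity, softening spores near the camera.
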Material /VehicleCosmetics/Boosts/Boost_Whisp/Materials/M_Boost_Whisp_PoisonSpore
graph {
 properties {
  bAutomaticallySetUsageInEditor=false
  bUsedWithNiagaraSprites=true
  bUsedWithParticleSprites=true
  blendMode=BLEND_Translucent
  shadingModel=MSM_Unlit
  usedShadingModels=Unlit
 }
 nodes [
  MaterialExpressionVertexColor_0 = ParticleColor()
  n1 = MaterialFunctionCall(Radius=ScalarParameter() {defaultValue=0.5, parameterName=Radius}) {materialFunction=&RadialGradientExponential}
  MaterialFunctionCall(Emissive=Multiply(A=@MaterialExpressionVertexColor_0.RGB, B=@n1.RadialGradientExponential)) {materialFunction=&MF_FX_EmissiveAdjustments}
  n2 = Constant() {r=80}
  n3 = Constant() {r=50}
  MaterialFunctionCall(Opacity In=Multiply(
    A=Multiply(A=@MaterialExpressionVertexColor_0.A, B=@n1.RadialGradientExponential)
    B=MaterialFunctionCall(Fade Length=@n2, Fade Offset=@n3) {materialFunction=&CameraDepthFade}.Result
   )) {materialFunction=&MF_VCVFX_DepthFade}
 ]
}
