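# Translucent unlit Niagara sprite material for a spark flare effect in the Whisp vehicle boost.
# A panning noise texture's red and green channels, scaled by the Distort dynamic parameter, offset the spark texture's UVs.
# That texture's green channel is tinted by particle color through MF_FX_EmissiveAdjustments and also feeds a separate saturated mask.
# Dissolve and Glow are named dynamic parameters, but the graph shows no direct use of them.
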Material /VehicleCosmetics/Boosts/Boost_Whisp/Materials/M_Boost_Whisp_SparkFlare
graph {
 properties {
  bAutomaticallySetUsageInEditor=false
  bUsedWithNiagaraSprites=true
  blendMode=BLEND_Translucent
  shadingModel=MSM_Unlit
  usedShadingModels=Unlit
 }
 nodes [
  n1 = TextureCoordinate()
  n2 = ComponentMask(None=TextureSample(UVs=Panner(Coordinate=@n1) {speedX=-0.6, speedY=0.5}) {texture=/DealershipSharedLibrary/FX/Textures/Noise/T_VCVFX_TrailMask_RGB}.RGB) {g=true, r=true}
  n3 = DynamicParameter() {paramNames [Dissolve,Glow,Distort,Param4]}
  n4 = TextureSampleParameter2D(UVs=Add(A=TextureCoordinate(), B=Multiply(A=@n2, B=@n3.Distort))) {parameterName=MaterialExpressionTextureSampleParameter2D_0, texture=/VehicleCosmetics/Boosts/Boost_Whisp/Textures/SparkJubile02_RGB_T}
  MaterialFunctionCall(Emissive=Multiply(A=ParticleColor().RGB, B=@n4.G)) {materialFunction=&MF_FX_EmissiveAdjustments}
  Saturate(None=Multiply(A=@n4.G))
 ]
}
