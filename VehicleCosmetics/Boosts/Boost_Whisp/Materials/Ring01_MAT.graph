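# Masked, unlit, two-sided particle material for a ring effect in the Whisp vehicle boost cosmetic.
# A sine-shaped exponential radial gradient forms the ring, multiplied by panning radial-UV fire noise and smoothstepped against the particle Dissolve parameter.
# Two panning sphere-mask layers, pulsed by clamped time sine and cosine, modulate a Spacetear space texture added to the ring.
# A scrolling gradient scaled by GradientGlow tints this sum before emissive adjustment.
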Material /VehicleCosmetics/Boosts/Boost_Whisp/Materials/Ring01_MAT
graph {
 properties {
  bAutomaticallySetUsageInEditor=false
  bUsedWithMeshParticles=true
  bUsedWithNiagaraMeshParticles=true
  bUsedWithNiagaraRibbons=true
  bUsedWithNiagaraSprites=true
  blendMode=BLEND_Masked
  shadingModel=MSM_Unlit
  twoSided=true
  usedShadingModels=Unlit
 }
 nodes [
  MaterialExpressionMeshEmitterDynamicParameter_0 = DynamicParameter() {paramNames [Dissolve,GradientGlow,Param3,Param4]}
  n1 = ScalarParameter() {parameterName=Max}
  n2 = Sine(None=MaterialFunctionCall(Radius=ScalarParameter() {defaultValue=0.5, parameterName=Radius}, Density=ScalarParameter() {defaultValue=2, parameterName=Density}) {materialFunction=&RadialGradientExponential}.RadialGradientExponential) {period=2}
  n3 = MaterialFunctionCall() {materialFunction=&VectorToRadialValue}
  n4 = AppendVector(A=ScalarParameter() {defaultValue=1, parameterName=TileX}, B=ScalarParameter() {defaultValue=1, parameterName=TileY})
  n5 = TextureSample(UVs=Panner(Coordinate=Multiply(A=@n3.Radial Coordinates, B=@n4)) {speedX=-0.2, speedY=-0.3}) {texture=/VehicleCosmetics/Boosts/Boost_Whisp/Textures/Dissolve_Fire_M_T}
  MaterialExpressionConstantClamp_0 = Saturate(None=SmoothStep(
    Min=@MaterialExpressionMeshEmitterDynamicParameter_0.Dissolve
    Max=Add(A=@MaterialExpressionMeshEmitterDynamicParameter_0.Dissolve, B=@n1)
    Value=Multiply(A=@n2, B=@n5.RGB)
   ))
  MaterialExpressionConstantClamp_2 = Saturate(None=Sine(None=Time()) {period=0.8})
  n6 = TextureSample(UVs=Panner(Coordinate=TextureCoordinate() {uTiling=1.5, vTiling=1.5}) {speedX=-0.8, speedY=0.05}) {texture=/DealershipSharedLibrary/FX/Textures/Gradient_Masks/T_VCVFX_SphereMaskTimeOffset_A}
  MaterialExpressionConstantClamp_3 = Saturate(None=Cosine(None=Time()) {period=0.8})
  n7 = TextureSample(UVs=Panner(Coordinate=TextureCoordinate() {uTiling=2, vTiling=2}) {speedX=-0.45, speedY=0.2}) {texture=&T_VCVFX_SphereMaskTimeOffset_A}
  n8 = Add(
   A=Add(A=Multiply(A=@MaterialExpressionConstantClamp_2, B=@n6.R), B=Multiply(A=@MaterialExpressionConstantClamp_3, B=@n7.R))
   B=Add(A=Multiply(A=@MaterialExpressionConstantClamp_2, B=@n6.B), B=Multiply(A=@n7.B, B=@MaterialExpressionConstantClamp_3))
  )
  MaterialExpressionConstantClamp_1 = Saturate(None=@n8)
  n9 = TextureSample(UVs=Panner(Coordinate=TextureCoordinate()) {speedX=-0.8}) {texture=/VehicleCosmetics/Boosts/Boost_Whisp/Textures/JubiGradient_T}
  n10 = Panner(Coordinate=@n3.Radial Coordinates) {speedY=0.1}
  n11 = Multiply(
   A=Clamp(None=@n8) {maxDefault=3, minDefault=0.1}
   B=Multiply(
    A=Constant() {r=20}
    B=TextureSample(UVs=@n10) {texture=/VehicleCosmetics/Boosts/Boost_Spacetear/Textures/T_VCVFX_boost_Space_Tear_Space}.RGB
   )
  )
  MaterialFunctionCall(Emissive=Multiply(A=Multiply(A=@n9.RGB, B=@MaterialExpressionMeshEmitterDynamicParameter_0.GradientGlow), B=Add(A=@n11, B=@n2))) {materialFunction=&MF_FX_EmissiveAdjustments}
 ]
}
